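; Movement mode extension logic for WallStick (Infinite variant), which lets a falling player stick to a wall, with activation blocked while a WallStick extension is already pushed.
; Helpers gate targeting on falling, held input and clambering tags, validate wall normals by angle, and record state in runtime data before notifying the pawn component.
; The physics and pre-movement updates are empty, jump adaptation passes through, and an older obstacle-aware jump-direction calculation remains.
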(fn BP_AdaptJumpDirectionAndType
  (params (WallStickRuntimeData :type Fort%20Movement%20Mode%20Wall%20Stick%20Runtime%20Data%20Object%20Reference) (WallNormal :type Vector%20%28by%20ref%29 :default "0, 0, 0") (InJumpDirection :type Vector :default "0, 0, 0") (InJumpType :type EWallJumpInitiatorType%20Enum))
  (return InJumpDirection InJumpType))

(fn BP_CanActivateExtension
  (return (not (Class|FortMovementCompCharacter|IsExtensionPushedwithTag (Variables|FortMovementMode_BaseExtLogic|GetOwnerMovementComponent) (GameplayTags|MakeLiteralGameplayTagContainer "(GameplayTags=((TagName=\"MovementMode.WallStick\")))")))))

(fn BP_PhysUpdate
  (params (RuntimeContext :type MMERuntime%20Context%20Structure%20%28by%20ref%29) (InVelocity :type Vector%20%28by%20ref%29 :default "0, 0, 0") (InAcceleration :type Vector%20%28by%20ref%29 :default "0, 0, 0")))

(fn BP_UpdateBeforeCharacterMovement
  (params (RuntimeContext :type MMERuntime%20Context%20Structure%20%28by%20ref%29)))

(fn CacheRuntimeData
  (params (Context :type MMERuntime%20Context%20Structure%20%28by%20ref%29))
  (bind _returnvalue (CallFunction|GetRuntimeData Context))
  (bind _output_get (Variables|Default|SetCurrentRuntimeData _returnvalue))
  (return _output_get))

(fn CalcJumpDirection_Old
  (params (WallStickRuntimeData :type Fort%20Movement%20Mode%20Wall%20Stick%20Runtime%20Data%20Object%20Reference) (WallNormal :type Vector%20%28by%20ref%29 :default "0, 0, 0"))
  (bind _ownermovementcomponent (Variables|FortMovementMode_BaseExtLogic|GetOwnerMovementComponent))
  (bind _returnvalue (Pawn|Components|CharacterMovement|GetCurrentAcceleration _ownermovementcomponent))
  (if (Math|Vector|VectorIsNearlyZero _returnvalue)
    (return (Variables|GetWallNormal))
    (else
      (bind _output_get (LocalVariables|SetNormalizedAcceleration (Math|Vector|Normalize _returnvalue)))
      (if (< (Math|Trig|Acos(Degrees) (Math|Vector|DotProduct _output_get (Math|Vector|NegateVector (Variables|GetWallNormal)))) 85.0)
        (bind _outhit (Class|BWallStickPawnComponent|CheckObstacleAbove (Actor|GetComponentbyClass (Variables|FortMovementMode_BaseExtLogic|GetOwnerPawn) "/WallStick/Components/B_WallStick_PawnComponent.B_WallStick_PawnComponent_C")))
        (if (Collision|BreakHitResult _outhit)
          ; For now, if we have obstruction above we're just going straight backwards. Feels like we'll want an extra case for obstacles that are above you but not immediately, where we have enough room to jump at a 45 degree angle
          (return (Variables|GetWallNormal))
          (else
            ; If we are unobstructed, do a (mostly) straight up jump
            (return (Math|Vector|Normalize (Math|Vector|Lerp(Vector) (Variables|GetWallNormal) "0.000000,0.000000,1.000000" 0.9)))))
        (else
          (return (LocalVariables|GetNormalizedAcceleration)))))))

(fn CanApplyHolster
  (params (HolsterId :type Name%20%28by%20ref%29 :default "None"))
  (bind _holsterid (Variables|Default|GetHolsterId))
  (bind _returnvalue (== HolsterId _holsterid))
  (bind _ownerpawn (Variables|FortMovementMode_BaseExtLogic|GetOwnerPawn))
  (Utilities|IsValid _ownerpawn
    (:Is%20Valid
      (return (and (not (GameplayTags|HasAnyMatchingGameplayTags _ownerpawn (LocalVariables|GetPreventHolster))) _returnvalue)))
    (:Is%20Not%20Valid
      (return _returnvalue))))

(fn CanTarget
  (bind (_velocity_x _velocity_y _velocity_z) (Math|Vector|BreakVector (Variables|/Script/Engine.MovementComponent|GetVelocity_X (Variables|FortMovementMode_BaseExtLogic|GetOwnerMovementComponent))))
  (return (Math|Boolean|ANDBoolean (< _velocity_z 0) (Variables|/WallStick/Proto/MME_RuntimeData_WallStick.MME_RuntimeData_WallStick_C|GetbInputHeld (Variables|Default|GetCurrentRuntimeData)) (Movement|IsFalling (Variables|FortMovementMode_BaseExtLogic|GetOwnerMovementComponent)) (not (GameplayTags|HasAnyMatchingGameplayTags (Variables|FortMovementMode_BaseExtLogic|GetOwnerPawn) (GameplayTags|MakeLiteralGameplayTagContainer "(GameplayTags=((TagName=\"Clambering.State.InActivationRange\"),(TagName=\"Clambering.State.LedgeMoving\"),(TagName=\"Clambering.State.WindowMoving\")))"))))))

(fn GetGravityAcceleration
  (return (Math|Vector|MakeVector 0.0 0.0 (Components|Movement|GetGravityZ (Variables|FortMovementMode_BaseExtLogic|GetOwnerMovementComponent)))))

(fn GetLocalAccelerationDirection
  (bind (_returnvalue_roll _returnvalue_pitch _returnvalue_yaw) (Math|Rotator|BreakRotator (Pawn|GetControlRotation (Variables|FortMovementMode_BaseExtLogic|GetOwnerPawn) :!class /Script/Engine.Pawn)))
  (return (Math|Vector|UnrotateVector (Math|Vector|Normalize (Pawn|Components|CharacterMovement|GetCurrentAcceleration (Variables|FortMovementMode_BaseExtLogic|GetOwnerMovementComponent))) (Math|Rotator|MakeRotator 0.0 0.0 _returnvalue_yaw))))

(fn GetRuntimeData
  (params (Context :type MMERuntime%20Context%20Structure%20%28by%20ref%29))
  (bind _returnvalue (Class|FortMovementModeExtBlueprintLibrary|GetRuntimeDatafromContext Context "/WallStick/Proto/MME_RuntimeData_WallStick.MME_RuntimeData_WallStick_C"))
  (return _returnvalue))

(fn IsAcceleratingForward
  (return (> (.x (CallFunction|GetLocalAccelerationDirection)) 0)))

(fn IsAcceleratingTowardsWall
  (return (< (Math|Vector|DotProduct (Math|Vector|Normalize (Pawn|Components|CharacterMovement|GetCurrentAcceleration (Variables|FortMovementMode_BaseExtLogic|GetOwnerMovementComponent))) (Variables|/WallStick/Proto/MME_RuntimeData_WallStick.MME_RuntimeData_WallStick_C|GetWallNormal (Variables|Default|GetCurrentRuntimeData))) (Variables|Default|GetMaxNormalAngle))))

(fn IsWallNormalValid
  (params (Normal :type Vector))
  (return (and (< (Math|Trig|Acos(Degrees) (Math|Vector|DotProduct (Math|Vector|NegateVector Normal) (Transformation|GetActorForwardVector (Variables|FortMovementMode_BaseExtLogic|GetOwnerPawn)))) (Variables|Default|GetMaxNormalAngle)) (Math|Float|InRange(Float) (Math|Trig|Acos(Degrees) (Math|Vector|DotProduct Normal "0.000000,0.000000,1.000000")) (Variables|Default|GetUpAngleMin) (Variables|Default|GetUpAngleMax)))))

(fn ShouldEnd
  (params (self2 :type MME%20Runtime%20Data%20Wall%20Stick%20Object%20Reference))
  (return (or (Variables|/WallStick/Proto/MME_RuntimeData_WallStick.MME_RuntimeData_WallStick_C|GetbTouchedGround self2) (Utilities|Enum|Equal(Enum) (Variables|/Script/Engine.CharacterMovementComponent|GetMovementMode (Variables|FortMovementMode_BaseExtLogic|GetOwnerMovementComponent)) "MOVE_Walking"))))

(fn StartSticking
  (params (Normal :type Vector))
  (bind _currentruntimedata (Variables|Default|GetCurrentRuntimeData))
  (Variables|/WallStick/Proto/MME_RuntimeData_WallStick.MME_RuntimeData_WallStick_C|SetWallNormal Normal _currentruntimedata)
  (Variables|/WallStick/Proto/MME_RuntimeData_WallStick.MME_RuntimeData_WallStick_C|SetCurrentState "NewEnumerator1" _currentruntimedata)
  (Development|PrintString (Variables|FortMovementMode_BaseExtLogic|GetOwnerPawn) "Sticking")
  (Class|BWallStickPawnComponent|HandleWallStickSucceeded (Variables|Default|GetWallStickComponent)))
